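# Masked Default Lit surface material for the decal on the SandTote vehicle cosmetic body.
# It holds a base color texture sample, a normal map sample and scalar parameters for Metallic, Specular and Roughness, with no wiring between them shown.
# It is flagged for skeletal meshes and mesh deformers, with automatic usage flagging and ray-traced shadow casting turned off.
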Material /VehicleCosmetics/Bodies/SandTote/Materials/M_SandTote_Decal
graph {
 properties {
  bAutomaticallySetUsageInEditor=false
  bCastRayTracedShadows=false
  bUsedWithMeshDeformer=true
  bUsedWithSkeletalMesh=true
  blendMode=BLEND_Masked
  usedShadingModels=Default Lit
 }
 nodes [
  ScalarParameter() {defaultValue=1, parameterName=Metallic}
  ScalarParameter() {defaultValue=0.5, parameterName=Specular}
  ScalarParameter() {defaultValue=0.35, parameterName=Roughness}
  TextureSample() {texture=/VehicleCosmetics/Bodies/SandTote/Textures/Decal/T_SandTote_Decal_BC}
  TextureSample() {samplerType=SAMPLERTYPE_Normal, texture=/VehicleCosmetics/Bodies/SandTote/Textures/Decal/T_SandTote_Decal_N}
 ]
}
