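# Physics asset for the SandTote vehicle body cosmetic.
# It defines a single box collision shape on the root bone, offset forward and upward to cover the body.
# It has no constraints, so the body collides as one rigid piece.
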PhysicsAsset /VehicleCosmetics/Bodies/SandTote/Meshes/Body_SandTote_PhysicsDM
bodies {root [{shapeType=Box, extentX=444, extentY=216, extentZ=132, shapeName=root_box, center=<20 0 96>}]}
constraints []
